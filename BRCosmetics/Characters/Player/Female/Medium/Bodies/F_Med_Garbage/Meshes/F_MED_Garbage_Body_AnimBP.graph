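# Post-processes the female medium Garbage body cosmetic, following the attached parent mesh and adding secondary motion.
# The copied parent pose passes through a deform Control Rig, then a constraint weights the arm bandana between two right upper-arm twist bones.
# Chained AnimDynamics nodes spring the bandana, its tips, boot strings and tags, with the tags partially blended, LOD-limited and kept off the chest by a spine_04 spherical limit.
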AnimBlueprint /BRCosmetics/Characters/Player/Female/Medium/Bodies/F_Med_Garbage/Meshes/F_MED_Garbage_Body_AnimBP
target_skeleton=&/F_MED_Garbage_Body_MESH_Skeleton
graphs [
 AnimGraph {nodes [
   n1 = Control Rig(Source=Copy Pose From Mesh() {bUseAttachedParent=true}) {
    alphaBoolBlend {blendOption=Linear}
    controlRigAssetReference {blueprintRigClass=&/F_MED_DeformRig_Garbage_ControlRig.F_MED_DeformRig_Garbage_ControlRig_C}
   }
   n2 = "Constraint - dyn_arm_bandana (PP)"!(ComponentPose=Local To Component(LocalPose=@n1), Alpha=1, ConstraintWeights_0=0.5, ConstraintWeights_1:float=0.5) -> (Pose) {
    alphaBoolBlend {blendOption=Linear}
    boneToModify {boneName=dyn_arm_bandana}
    constraintSetup [{targetBone {boneName=upperarm_twist_01_r}}, {targetBone {boneName=upperarm_twist_02_r}}]
    constraintWeights [1,1]
   }
   n3 = Anim Dynamics - Bone: dyn_arm_bandana(ComponentPose=@n2, Alpha=1, ExternalForce="0,0,0") {
    angularDampingOverride=1
    angularSpringConstant=20000
    bAngularSpring=true
    bOverrideAngularDamping=true
    bUseGravityOverride=true
    gravityScale=0.35
    simulationSpace=RootRelative
    alphaBoolBlend {blendOption=Linear}
    boundBone {boneName=dyn_arm_bandana}
    componentLinearAccScale=<0.3 0.2 0.2>
    componentLinearVelScale=<0.3 0.2 0.2>
    gravityOverride {z=-980}
    physicsBodyDefinitions [{
      boundBone {boneName=dyn_arm_bandana}
      boxExtents=<20 30 30>
      constraintSetup {angularLimitsMax=<15 15 15>, angularLimitsMin=<-15 -15 -15>, angularTarget=<1 0 0>}
      localJointOffset=<4 0 0>
     }]
   }
   n4 = Anim Dynamics - Bone: dyn_arm_bandana_tip_upper(ComponentPose=@n3, Alpha=1, ExternalForce="0,0,0") {
    angularDampingOverride=1
    angularSpringConstant=20000
    bAngularSpring=true
    bOverrideAngularDamping=true
    bUseGravityOverride=true
    gravityScale=0.35
    simulationSpace=RootRelative
    alphaBoolBlend {blendOption=Linear}
    boundBone {boneName=dyn_arm_bandana_tip_upper}
    componentLinearAccScale=<0.3 0.2 0.2>
    componentLinearVelScale=<0.3 0.2 0.2>
    gravityOverride {z=-980}
    physicsBodyDefinitions [{
      boundBone {boneName=dyn_arm_bandana_tip_upper}
      boxExtents=<20 30 30>
      constraintSetup {angularLimitsMax=<80 80 80>, angularLimitsMin=<-80 -80 -80>, angularTarget=<1 0 0>}
      localJointOffset=<4 0 0>
     }]
   }
   n5 = Anim Dynamics - Bone: dyn_arm_bandana_tip_lower(ComponentPose=@n4, Alpha=1, ExternalForce="0,0,0") {
    angularDampingOverride=1
    angularSpringConstant=20000
    bAngularSpring=true
    bOverrideAngularDamping=true
    bUseGravityOverride=true
    gravityScale=0.35
    simulationSpace=RootRelative
    alphaBoolBlend {blendOption=Linear}
    boundBone {boneName=dyn_arm_bandana_tip_lower}
    componentLinearAccScale=<0.3 0.2 0.2>
    componentLinearVelScale=<0.3 0.2 0.2>
    gravityOverride {z=-980}
    physicsBodyDefinitions [{
      boundBone {boneName=dyn_arm_bandana_tip_lower}
      boxExtents=<20 30 30>
      constraintSetup {angularLimitsMax=<80 80 80>, angularLimitsMin=<-80 -80 -80>, angularTarget=<1 0 0>}
      localJointOffset=<4 0 0>
     }]
   }
   n6 = Anim Dynamics - Bone: dyn_boot_strings_01(ComponentPose=@n5, Alpha=1, ExternalForce="0,0,0") {
    angularDampingOverride=1
    angularSpringConstant=20000
    bAngularSpring=true
    bOverrideAngularDamping=true
    bUseGravityOverride=true
    gravityScale=0.35
    simulationSpace=BoneRelative
    alphaBoolBlend {blendOption=Linear}
    boundBone {boneName=dyn_boot_strings_01}
    componentLinearAccScale=<0.3 0.2 0.2>
    componentLinearVelScale=<0.3 0.2 0.2>
    gravityOverride {z=-980}
    physicsBodyDefinitions [{
      boundBone {boneName=dyn_boot_strings_01}
      boxExtents=<20 30 30>
      constraintSetup {angularLimitsMax=<60 15 30>, angularLimitsMin=<-60 -15 -60>, angularTarget=<1 0 0>}
      localJointOffset=<0 0 -10>
     }]
    relativeSpaceBone {boneName=root}
   }
   n7 = Anim Dynamics - Bone: dyn_boot_strings_02(ComponentPose=@n6, Alpha=1, ExternalForce="0,0,0") {
    angularDampingOverride=1
    angularSpringConstant=20000
    bAngularSpring=true
    bOverrideAngularDamping=true
    bUseGravityOverride=true
    gravityScale=0.35
    simulationSpace=BoneRelative
    alphaBoolBlend {blendOption=Linear}
    boundBone {boneName=dyn_boot_strings_02}
    componentLinearAccScale=<0.3 0.2 0.2>
    componentLinearVelScale=<0.3 0.2 0.2>
    gravityOverride {z=-980}
    physicsBodyDefinitions [{
      boundBone {boneName=dyn_boot_strings_02}
      boxExtents=<20 30 30>
      constraintSetup {angularLimitsMax=<60 15 30>, angularLimitsMin=<-60 -15 -60>, angularTarget=<1 0 0>}
      localJointOffset=<0 0 -10>
     }]
    relativeSpaceBone {boneName=root}
   }
   n8 = Anim Dynamics - Bone: dyn_tags(ComponentPose=@n7, Alpha=0.7, ExternalForce="0,0,0") {
    angularBiasOverride=1
    angularDampingOverride=0.9999
    angularSpringConstant=1000
    bAngularSpring=true
    bGravityOverrideInSimSpace=true
    bOverrideAngularDamping=true
    bUseSphericalLimits=true
    lODThreshold=1
    numSolverIterationsPostUpdate=4
    numSolverIterationsPreUpdate=16
    simulationSpace=RootRelative
    alphaBoolBlend {blendOption=Linear}
    boundBone {boneName=dyn_tags}
    physicsBodyDefinitions [{
      collisionType=CustomSphere
      sphereCollisionRadius=1.5
      boundBone {boneName=dyn_tags}
      boxExtents=<15 4 20>
      constraintSetup {angularTargetAxis=AxisZ, twistAxis=AxisZ, angularLimitsMax=<25 45 0>, angularLimitsMin=<0 -45 0>, angularTarget=<0 0 0>}
      localJointOffset=<0 0 -2>
     }]
    sphericalLimits [{limitRadius=11.5, drivingBone {boneName=spine_04}, sphereLocalOffset=<0 0 0>}]
   }
   Output Pose(Result=Component To Local(ComponentPose=@n8)) {layerGroup=None}
  ]}
 EventGraph {nodes [Event BlueprintUpdateAnimation()]}
]
variables []
