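# Physics asset for the F_MED_Garbage female character body mesh, set up for secondary motion of the front and left temple hair.
# Capsule bodies form a three-bone front hair chain and a single temple strand, both constrained to a spherical head body.
# The front chain's root swings and twists freely, its lower links allow moderate swing with almost no twist, and the temple strand has wider angular limits.
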PhysicsAsset /BRCosmetics/Characters/Player/Female/Medium/Bodies/F_Med_Garbage/Meshes/Parts/F_MED_Garbage_Physics
bodies {
 dyn_hairFront_01 [{shapeType=Capsule, length=0.314, radius=1.2202, rotation=<rot 0 90 90>}]
 dyn_hairFront_02 [{shapeType=Capsule, length=3.2218, radius=1.3938, center=<2.2044 -0.3287 0.0775>, rotation=<rot 0 87.8569 -88.977>}]
 dyn_hairFront_03 [{shapeType=Capsule, length=3.2481, radius=1.119, center=<1.8998 -0.2209 -0.2223>, rotation=<rot 0.329 -94.5933 -87.6067>}]
 dyn_hairTemple_l [{shapeType=Capsule, length=1.7458, radius=0.7535, center=<0.1971 0.6263 0.0112>, rotation=<rot 0 -17.4724 -92.1758>}]
 head [{shapeType=Sphere, radius=11.9569, center=<3.1424 -2.7193 0.3776>}]
}
constraints [
 {
  bone1Name=dyn_hairFront_01
  bone2Name=head
  swing1LimitDegrees=45
  swing1Motion=Free
  swing2LimitDegrees=45
  swing2Motion=Free
  twistLimitDegrees=45
  twistMotion=Free
 }
 {bone1Name=dyn_hairFront_02, bone2Name=dyn_hairFront_01, swing1LimitDegrees=35, swing2LimitDegrees=35, twistLimitDegrees=1}
 {bone1Name=dyn_hairFront_03, bone2Name=dyn_hairFront_02, swing1LimitDegrees=35, swing2LimitDegrees=35, twistLimitDegrees=1}
 {bone1Name=dyn_hairTemple_l, bone2Name=head, swing1LimitDegrees=45, swing2LimitDegrees=45, twistLimitDegrees=40}
]
