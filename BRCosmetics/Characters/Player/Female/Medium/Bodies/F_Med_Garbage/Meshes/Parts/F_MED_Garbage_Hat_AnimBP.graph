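# Anim Blueprint for the female Garbage outfit's hat mesh, which adds secondary hair-bone motion on a pose copied from the attached parent mesh.
# Flap nodes drive base_hair_back from head rotation, then Anim Dynamics simulate both side hair bones with neck sphere limits.
# Bone offsets follow, one weighted by upward aim while targeting, before a partial RigidBody pass.
# The event graph also computes other head-direction and speed values the AnimGraph never reads.
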AnimBlueprint /BRCosmetics/Characters/Player/Female/Medium/Bodies/F_Med_Garbage/Meshes/Parts/F_MED_Garbage_Hat_AnimBP
target_skeleton=/Game/Characters/Player/Female/Medium/Bodies/F_Med_Garbage/Meshes/Parts/F_MED_Garbage_Hat_MESH_Skeleton
graphs [
 AnimGraph {nodes [
   n1 = Flap Node - Bone: base_hair_back!(
    ComponentPose:Component Space Pose Link Structure=Local To Component(LocalPose=Copy Pose From Mesh() {bUseAttachedParent=true})
    Alpha:float=1
   ) -> (Pose:Component Space Pose Link Structure) {
    is=title FortAnimGraphNode_Flap
    bUseClamp=true
    targetClampMax=5
    targetClampMin=-1
    targetComponent=TranslationY
    alphaBoolBlend {blendOption=Linear}
    alphaScaleBiasClamp {bInterpResult=true}
    sourceDrives [{
      bUseRange=false
      drivingCurve=null
      multiplier=-2
      rangeMax=1
      rangeMin=0
      remappedMax=1
      remappedMin=0
      sourceComponent=RotationZ
      useQuaternion=true
      sourceBone {boneName=head}
     }]
    targetBone {boneName=base_hair_back}
   }
   n2 = Flap Node - Bone: base_hair_back(ComponentPose=@n1, Alpha=1) {
    bUseClamp=true
    targetClampMax=100
    targetComponent=TranslationX
    alphaBoolBlend {blendOption=Linear}
    alphaScaleBiasClamp {bInterpResult=true}
    sourceDrives [{
      bUseRange=false
      drivingCurve=null
      multiplier=5
      rangeMax=50
      rangeMin=0
      remappedMax=80
      remappedMin=0
      sourceComponent=RotationZ
      useQuaternion=true
      sourceBone {boneName=head}
     }]
    targetBone {boneName=base_hair_back}
   }
   n3 = Flap Node - Bone: base_hair_back(ComponentPose=@n2, Alpha=1) {
    bUseClamp=true
    targetClampMax=5
    targetClampMin=-1
    targetComponent=RotationY
    alphaBoolBlend {blendOption=Linear}
    alphaScaleBiasClamp {bInterpResult=true}
    sourceDrives [{
      bUseRange=false
      drivingCurve=null
      multiplier=-0.25
      rangeMax=50
      rangeMin=0
      remappedMax=50
      remappedMin=0
      sourceComponent=RotationZ
      useQuaternion=true
      sourceBone {boneName=head}
     }]
    targetBone {boneName=base_hair_back}
   }
   n4 = Flap Node - Bone: base_hair_back(ComponentPose=@n3, Alpha=1) {
    bUseClamp=true
    targetClampMax=10
    targetComponent=RotationZ
    alphaBoolBlend {blendOption=Linear}
    alphaScaleBiasClamp {bInterpResult=true}
    sourceDrives [{
      bUseRange=false
      drivingCurve=null
      multiplier=1.5
      rangeMax=40
      rangeMin=0
      remappedMax=80
      remappedMin=0
      sourceComponent=RotationZ
      useQuaternion=true
      sourceBone {boneName=head}
     }]
    targetBone {boneName=base_hair_back}
   }
   n5 = Anim Dynamics - Bone: dyn_hair_r!(ComponentPose=@n4, Alpha=1, AngularSpringConstant:float=4000, ExternalForce="0,0,0") -> (Pose) {
    angularDampingOverride=0.9999
    angularSpringConstant=1000
    bAngularSpring=true
    bGravityOverrideInSimSpace=true
    bOverrideAngularDamping=true
    bUsePlanarLimit=false
    bUseSphericalLimits=true
    gravityScale=0.35
    alphaBoolBlend {blendOption=Linear}
    boundBone {boneName=dyn_hair_r}
    componentAppliedLinearAccClamp=<70 70 70>
    componentLinearVelScale=<300 300 300>
    gravityOverride {z=-980}
    physicsBodyDefinitions [{
      collisionType=CustomSphere
      sphereCollisionRadius=3
      boundBone {boneName=dyn_hair_r}
      boxExtents=<3 20 20>
      constraintSetup {angularLimitsMax=<5 10 20>, angularLimitsMin=<-5 -10 -40>, angularTarget=<1 0 0>}
      localJointOffset=<3 0 0>
     }]
    sphericalLimits [{limitRadius=2.5, drivingBone {boneName=neck_02}, sphereLocalOffset=<0 0 0>}]
   }
   n6 = Anim Dynamics - Bone: dyn_hair_l!(ComponentPose=@n5, Alpha=1, AngularSpringConstant:float=4000, ExternalForce="0,0,0") -> (Pose) {
    angularDampingOverride=1
    angularSpringConstant=1000
    bAngularSpring=true
    bGravityOverrideInSimSpace=true
    bOverrideAngularDamping=true
    bUsePlanarLimit=false
    gravityScale=0.35
    alphaBoolBlend {blendOption=Linear}
    boundBone {boneName=dyn_hair_l}
    componentAppliedLinearAccClamp=<70 70 70>
    componentLinearVelScale=<300 300 300>
    gravityOverride {z=-980}
    physicsBodyDefinitions [{
      collisionType=CustomSphere
      sphereCollisionRadius=3
      boundBone {boneName=dyn_hair_l}
      boxExtents=<3 20 20>
      constraintSetup {angularLimitsMax=<0 10 20>, angularLimitsMin=<0 -10 -65>, angularTarget=<0 0 0>}
      localJointOffset=<2 0 0>
     }]
    sphericalLimits [{limitRadius=5, drivingBone {boneName=neck_02}, sphereLocalOffset=<0 0 0>}]
   }
   n7 = "Transform (Modify) Bone - Bone: dyn_hair_l"(ComponentPose=@n6, Alpha=1, Translation="0,-1.5,0", Rotation="0,0,0", Scale="1,1,1") {
    translationMode=BMM_Additive
    translationSpace=BCS_BoneSpace
    alphaBoolBlend {blendOption=Linear}
    boneToModify {boneName=dyn_hair_l}
    translation {y=-1.5}
   }
   n8 = "Transform (Modify) Bone - Bone: base_hair_back"(ComponentPose=@n7, Alpha=1, Translation="-0.117,-0.961,0.000", Rotation="-0.000,-6.177,-0.001", Scale="1,1,1") {
    rotationMode=BMM_Additive
    rotationSpace=BCS_BoneSpace
    translationMode=BMM_Additive
    translationSpace=BCS_BoneSpace
    alphaBoolBlend {blendOption=Linear}
    boneToModify {boneName=base_hair_back}
    rotation=<rot 0 -6.177 -0.001>
    translation=<-0.117 -0.961>
   }
   n9 = "Transform (Modify) Bone - Bone: base_hair_back"(
    ComponentPose=@n8
    Alpha=Get Is Aiming Up() -> (Is Aiming Up:double)
    Translation="-0.593,-0.802,0.457"
    Rotation="-0.992,6.339,23.458"
    Scale="1,1,1"
   ) {
    rotationMode=BMM_Additive
    rotationSpace=BCS_BoneSpace
    scaleMode=BMM_Additive
    scaleSpace=BCS_BoneSpace
    translationMode=BMM_Additive
    translationSpace=BCS_BoneSpace
    alphaBoolBlend {blendOption=Linear}
    alphaScaleBiasClamp {bInterpResult=true}
    boneToModify {boneName=base_hair_back}
    rotation=<rot -0.992 6.339 23.458>
    translation=<-0.593 -0.802 0.457>
   }
   n10 = "Transform (Modify) Bone - Bone: base_hair_back"!(ComponentPose=@n9, Alpha=1, Rotation="3,-8.5,0") -> (Pose) {
    rotationMode=BMM_Additive
    rotationSpace=BCS_BoneSpace
    scaleMode=BMM_Additive
    scaleSpace=BCS_BoneSpace
    translationMode=BMM_Additive
    translationSpace=BCS_BoneSpace
    alphaBoolBlend {blendOption=Linear}
    alphaScaleBiasClamp {bInterpResult=true, interpSpeedDecreasing=1, interpSpeedIncreasing=1}
    boneToModify {boneName=base_hair_back}
    rotation {pitch=3, yaw=-8.5}
   }
   n11 = "Transform (Modify) Bone - Bone: base_hair_back"(ComponentPose=@n10, Alpha=1, Rotation="3,-8.5,0") {
    rotationMode=BMM_Additive
    rotationSpace=BCS_BoneSpace
    scaleMode=BMM_Additive
    scaleSpace=BCS_BoneSpace
    translationMode=BMM_Additive
    translationSpace=BCS_BoneSpace
    alphaBoolBlend {blendOption=Linear}
    alphaScaleBiasClamp {bInterpResult=true, interpSpeedDecreasing=1, interpSpeedIncreasing=1}
    boneToModify {boneName=base_hair_back}
    rotation {pitch=3, yaw=-8.5}
   }
   n12 = RigidBody(ComponentPose=@n11, Alpha=0.4) {
    bOverrideWorldGravity=true
    overridePhysicsAsset=&/F_MED_Garbage_Physics
    simulationSpace=BaseBoneSpace
    alphaBoolBlend {blendOption=Linear}
    baseBoneRef {boneName=spine_04}
    componentAppliedLinearAccClamp=<100 100 100>
    componentLinearAccScale=<0.3 0.3 0.3>
    componentLinearVelScale=<0.3 0.3 0.3>
    overrideWorldGravity {z=-980}
   }
   Output Pose(Result=Component To Local(ComponentPose=@n12)) {layerGroup=None}
  ]}
 EventGraph {nodes [
   Event BlueprintUpdateAnimation()
   > n1 = Sequence()
   > Set Head Up(Head Up:float=GetAlphaFromReferenceTransformDelta(
     SkeletalMeshComponent="Get Skel Mesh "() -> ("Skel Mesh ":Skeletal Mesh Component Object Reference)
     BoneName=spine_05
     BaseName=head
     InputRange=(X=15,Y=60)
     TwistAxis="0,0,1"
    )) -> (Output_Get:float)
   n2 = "Get Skel Mesh "()
   > Set Head Down(Head Down:float=GetAlphaFromReferenceTransformDelta(SkeletalMeshComponent=@n2, BoneName=spine_05, BaseName=head, InputRange=(X=-15,Y=-60), OutputRange=(X=0,Y=-1), TwistAxis="0,0,1")) -> (Output_Get:float)
   n3 = "To Float (Boolean)"(InBool=Get bIsInFrontEnd())
   > Set Head Right(Head Right:float=float * float(
     A=GetAlphaFromReferenceTransformDelta(SkeletalMeshComponent=@n2, BoneName=head, BaseName=neck_02, DeltaComponentType=EulerX, InputRange=(X=0,Y=-1), TwistAxis="1,0,0")
     B=@n3
    ) {is=title K2Node_PromotableOperator}) -> (Output_Get:float)
   > Set Head Left(Head Left:float=float * float(
     A=GetAlphaFromReferenceTransformDelta(SkeletalMeshComponent=@n2, BoneName=head, BaseName=neck_02, DeltaComponentType=EulerX, InputRange=(X=0,Y=1), TwistAxis="1,0,0")
     B=@n3
    )) -> (Output_Get:float)
   > n4 = Branch(Condition=Get bIsTargeting())
   > Set Is Aiming Up(Is Aiming Up:double=Get Head Up() -> (Head Up:float)) -> (Output_Get:double)
   > n5 = Branch(Condition=Get bIsTargeting())
   > Set Is Aiming Down(Is Aiming Down:double=Get Head Down() -> (Head Down:float)) -> (Output_Get:double)
   Set Is Aiming Down(execute=@n5.else, Is Aiming Down=0)
   Set Is Aiming Up(execute=@n4.else, Is Aiming Up=0)
   n6 = MapRangeClamped(Value=Get PawnSpeed2D(), InRangeA=0, InRangeB=100)
   n7 = Branch(execute=@n1.then_1, Condition="float <= float"!(A:double=@n6, B:double=0) -> (ReturnValue:bool) {is=title K2Node_CallFunction})
   > n8 = Set speed_mult(speed_mult:double=0.1) -> (Output_Get:double)
   > Set HairAlpha(HairAlpha:double=@n8) -> (Output_Get:double)
   Set speed_mult(execute=@n7.else, speed_mult=float * float(A=@n6, B=0.65) {is=title K2Node_CommutativeAssociativeBinaryOperator})
   Event BlueprintInitializeAnimation()
   > "Set Skel Mesh "("Skel Mesh ":Skeletal Mesh Component Object Reference=GetOwnerSkeletalMeshComponent()) -> (Output_Get:Skeletal Mesh Component Object Reference)
  ]}
]
variables [
 HairAlpha {type=number}
 Head Down {type=number}
 Head Left {type=number}
 Head Right {type=number}
 Head Up {type=number}
 Is Aiming Down {type=number}
 Is Aiming Up {type=number}
 "Skel Mesh " {type=SkeletalMeshComponent}
 speed_mult {type=number}
]
